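; BP_StarsoapUIDirector is a Dynamic UI director for the Starsoap proximity chat HUD.
; On BeginPlay it reads the DynamicAthenaHUDDirector.ExcludeProximityChatFromDefaultScenes console variable.
; When that variable is true, it adds its StarsoapProximityChatScene itself, supplying the chat scene that the variable's name indicates is left out of the default scenes.
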(event EventBeginPlay
  (bind _returnvalue (Development|GetConsoleVariableBoolValue "DynamicAthenaHUDDirector.ExcludeProximityChatFromDefaultScenes"))
  (if _returnvalue
    (DynamicUIDirector|AddScene (Variables|Default|GetStarsoapProximityChatScene))))

(fn ConstructionScript)
