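# Proximity chat HUD widget that shows a "PROXIMITY CHAT AVAILABLE" label above an active speakers container with proximity chat support enabled.
# UpdateIsInArea and UpdateIsTargetActive store their flags, then both call UpdateTextVisibility.
# That update shows or hides the label from the in-area flag alone; the target-active flag is stored but never read here.
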WidgetBlueprint /StarsoapGameplay/UI/ProximityChat/WBP_Starsoap_ProximityChat
graphs [EventGraph {nodes [
   n1 = UpdateIsInArea!() -> (OutputDelegate, then, IsInArea:bool=~) {is=title K2Node_CustomEvent}
   > n2 = Set Is in Area(Is in Area:bool=@n1.IsInArea) -> (Output_Get:bool)
   UpdateTextVisibility(execute=[@n2.then, @n3.then])
   UpdateTextVisibility() {is=title K2Node_CustomEvent}
   > SetVisibility(
    self=Get TEXT_ProximityChatAvailable() -> (TEXT_ProximityChatAvailable:Common Text Object Reference)
    InVisibility=Select(Option 0=Hidden, Option 1=SelfHitTestInvisible, Index=Get Is in Area() -> (Is in Area:bool))
   )
   n3 = Set Is Target Active(Is Target Active:bool=UpdateIsTargetActive!() -> (OutputDelegate, then, IsTargetActive:bool=~) {is=title K2Node_CustomEvent}.IsTargetActive) -> (Output_Get:bool)
  ]}]
variables [
 Is in Area {type=boolean}
 Is Target Active {type=boolean}
]
widgets {
 info {inheritedWidgetCount=0, parentClass=&UserWidget, rootWidgetClass=&ScaleBox, widgetCount=5}
 widgets [
  ScaleBox_1 {widgetClassPath=&ScaleBox, widget=ScaleBox_1, properties {stretch=UserSpecified, userSpecifiedScale=0.8}}
  VerticalBox {widgetClassPath=&VerticalBox, parent=ScaleBox_1, widget=VerticalBox, slot {class=ScaleBoxSlot}}
  TEXT_ProximityChatAvailable {
   widgetClassPath=&CommonTextBlock
   bIsVariable=true
   parent=VerticalBox
   widget=TEXT_ProximityChatAvailable
   properties {
    mobileFontSizeMultiplier=1.2
    scrollStyle=&TextScrollStyle-Base_C
    text=PROXIMITY CHAT AVAILABLE
    font {
     bForceMonospaced=false
     bMaterialIsStencil=false
     fontMaterial=null
     fontObject=&BurbankSmall
     letterSpacing=0
     monospacedWidth=1
     size=15
     skewAmount=0
     typefaceFontName=Black
     outlineSettings {
      bApplyOutlineToDropShadows=false
      bMiteredCorners=false
      bSeparateFillAlpha=false
      outlineMaterial=&M-FontShadow
      outlineSize=1
      outlineColor=<rgba 0 0 0 1>
     }
    }
    shadowOffset=<0 0>
   }
   slot {class=VerticalBoxSlot, properties {horizontalAlignment=HAlign_Center, verticalAlignment=VAlign_Center, padding=<ltrb 6 2 8 2>}}
  }
  HorizontalBox_VoiceIndicators {widgetClassPath=&HorizontalBox, parent=VerticalBox, widget=HorizontalBox_VoiceIndicators, slot {class=VerticalBoxSlot}}
  WBP_StarsoapActiveSpeakersContainer {
   widgetClassPath=&/WBP_StarsoapActiveSpeakersContainer
   bIsVariable=true
   parent=HorizontalBox_VoiceIndicators
   widget=WBP_StarsoapActiveSpeakersContainer
   properties {bShouldSupportProximityChat=true}
   slot {class=HorizontalBoxSlot}
  }
 ]
}
